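# Landscape material layer function that height-blends Hera asphalt onto an incoming material, with separate desktop and mobile paths.
# Default and Mobile quality-level sub-functions supply the asphalt, which opacity-mask height-blend utilities composite by layer weight and contrast, using a displacement-custom variant on desktop.
# An IsMeshPartition static switch bypasses the Asphalt landscape layer switch.
# Outputs include both blended materials, a layer mask and a static-mesh material with derived normals on mobile.
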MaterialFunction /Hera/Environment/Terrain/Landscape/Materials/Layers/MF_Hera_Asphalt
graph {nodes [
  n1 = NamedRerouteUsage()
  n2 = NamedRerouteUsage()
  n3 = NamedRerouteUsage()
  n4 = NamedRerouteUsage()
  n5 = NamedRerouteUsage()
  n6 = MaterialFunctionCall!(InputMaterial=@n1, LayerMaterial=@n2, LayerWeight=@n3, UseContrast=@n4, Contrast=@n5) {materialFunction=/Game/Athena/Helios/Environments/Landscape/Materials/MaterialFunctions_Utilities/MF_HeightBlendLandscapeLayers_OpacityMaskDisplaceCustom}
  n7 = StaticSwitchParameter(True=@n6.BlendedMaterial, False=LandscapeLayerSwitch!(LayerUsed=@n6.BlendedMaterial, LayerNotUsed=@n1) {parameterName=Asphalt}) {
   desc="Bypass landscape layer switch \r\nif used on MeshPartition\r\n\r\nMeshPartitionWeightChannelSwitch \r\nwill be available eventually"
   parameterName=IsMeshPartition
  }
  FunctionOutput(None=@n7) {outputName=BlendedMaterial, sortPriority=-300}
  FunctionOutput(None=ShadingPathSwitch(Default=NamedRerouteUsage(), Mobile=NamedRerouteUsage())) {outputName=LayerMask, sortPriority=-1}
  n8 = NamedRerouteUsage()
  n9 = MaterialFunctionCall!(
   InputMaterial=@n8
   LayerMaterial=NamedRerouteUsage()
   LayerWeight=NamedRerouteUsage()
   UseContrast=NamedRerouteUsage()
   Contrast=NamedRerouteUsage()
  ) {materialFunction=/Game/Athena/Helios/Environments/Landscape/Materials/MaterialFunctions_Utilities/MF_HeightBlendLandscapeLayers_OpacityMask}
  n10 = StaticSwitchParameter(True=@n9.BlendedMaterial, False=LandscapeLayerSwitch(LayerUsed=@n9.BlendedMaterial, LayerNotUsed=@n8) {parameterName=Asphalt}) {
   desc="Bypass landscape layer switch \r\nif used on MeshPartition\r\n\r\nMeshPartitionWeightChannelSwitch \r\nwill be available eventually"
   parameterName=IsMeshPartition
  }
  FunctionOutput(None=@n10) {outputName=MobileBlendedMaterial, sortPriority=-100}
  n11 = NamedRerouteUsage()
  n12 = StaticBool()
  FunctionOutput(None=ShadingPathSwitch(
    Default=NamedRerouteUsage()
    Mobile=MaterialFunctionCall!(MobileMaterial=@n11, TransformTangents=@n12) {materialFunction=/Game/Athena/Helios/Environments/Landscape/Materials/MaterialFunctions_Utilities/MF_MobileDeriveNormals}.ResultMaterial
   )) {outputName=StaticMeshMaterial, sortPriority=10}
  MaterialFunctionCall(InputMaterial=@n1, LayerMaterial=@n2, LayerWeight=@n3, UseContrast=@n4, Contrast=@n5) {materialFunction=&MF_HeightBlendLandscapeLayers_OpacityMask}
  NamedRerouteDeclaration(None=@n6.LayerMask) {name=LayerMask, nodeColor=<rgba 0 1 0.6953 1>}
  NamedRerouteDeclaration(None=FunctionInput(Preview=Constant()) {bUsePreviewValueAsDefault=true, inputName=VariationMasks Mobile, inputType=FunctionInput_Vector4, sortPriority=40}) {name=VariationMasks Mobile, nodeColor=<rgba 1 0.8498 0.0704 1>}
  n13 = SetMaterialAttributes!(MaterialAttributes)
  NamedRerouteDeclaration(None=FunctionInput(Preview=@n13) {bUsePreviewValueAsDefault=true, inputName=Material, inputType=FunctionInput_MaterialAttributes}) {name=InputMaterial, nodeColor=<rgba 0 1 0 1>}
  NamedRerouteDeclaration(None=MaterialFunctionCall!(LayerWeight=NamedRerouteUsage(), UseTangentNormals=NamedRerouteUsage()) {materialFunction=/Hera/Environment/Terrain/Landscape/Materials/Layers/QualityLevel/MF_Hera_Asphalt_Default}.Result) {name=LayerMaterial, nodeColor=<rgba 0 1 0 1>}
  NamedRerouteDeclaration(None=FunctionInput(Preview=Constant()) {bUsePreviewValueAsDefault=true, inputName=LayerWeight, inputType=FunctionInput_Scalar, sortPriority=15}) {name=LayerWeight, nodeColor=<rgba 0.5 0 1 1>}
  NamedRerouteDeclaration(None=@n9.LayerMask) {name=MobileLayerMask, nodeColor=<rgba 0.7813 0 1 1>}
  NamedRerouteDeclaration(None=FunctionInput(Preview=Constant()) {bUsePreviewValueAsDefault=true, inputName=TerrainMasks_QuadA, inputType=FunctionInput_Vector4, sortPriority=20}) {name=TerrainMasks_QuadA, nodeColor=<rgba 0.5 0 1 1>}
  NamedRerouteDeclaration(None=FunctionInput(Preview=Constant()) {bUsePreviewValueAsDefault=true, inputName=TerrainMasks_QuadC, inputType=FunctionInput_Vector4, sortPriority=30}) {name=TerrainMasks_QuadC, nodeColor=<rgba 0.5 0 1 1>}
  NamedRerouteDeclaration(None=FunctionInput(Preview=Constant()) {bUsePreviewValueAsDefault=true, inputName=TerrainMasks_QuadD, inputType=FunctionInput_Vector4, sortPriority=35}) {name=TerrainMasks_QuadD, nodeColor=<rgba 0.5 0 1 1>}
  NamedRerouteDeclaration(None=FunctionInput(Preview=Constant()) {bUsePreviewValueAsDefault=true, inputName=TerrainMasks_QuadB, inputType=FunctionInput_Vector4, sortPriority=25}) {name=TerrainMasks_QuadB, nodeColor=<rgba 0.5 0 1 1>}
  NamedRerouteDeclaration(None=FunctionInput(Preview=@n13) {bUsePreviewValueAsDefault=true, inputName=MobileMaterial, inputType=FunctionInput_MaterialAttributes, sortPriority=10}) {name=MobileInputMaterial, nodeColor=<rgba 0 0.5 1 1>}
  NamedRerouteDeclaration(None=MaterialFunctionCall!(LayerWeight=NamedRerouteUsage(), VariationMasks Mobile=NamedRerouteUsage()) {materialFunction=/Hera/Environment/Terrain/Landscape/Materials/Layers/QualityLevel/MF_Hera_Asphalt_Mobile}.Result) {name=MobileLayerMaterial, nodeColor=<rgba 0 0.5 1 1>}
  NamedRerouteDeclaration(None=StaticBool() {value=true}) {name=UseContrast, nodeColor=<rgba 0.5 0 1 1>}
  NamedRerouteDeclaration(None=Constant() {r=0.5}) {name=LayerBlendContrast, nodeColor=<rgba 0.5 0 1 1>}
  n14 = FunctionInput(Preview=StaticBool()) {
   bUsePreviewValueAsDefault=true
   description="Default = False"
   inputName=UseTangentNormals
   inputType=FunctionInput_StaticBool
   sortPriority=15
  }
  NamedRerouteDeclaration(None=@n14) {name=UseTangentNormalsBool, nodeColor=<rgba 0.5 0 1 1>}
 ]}
