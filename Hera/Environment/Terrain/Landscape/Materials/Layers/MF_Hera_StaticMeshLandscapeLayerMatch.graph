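# Material function that rebuilds Hera landscape layers for static meshes to match the terrain, with world-space normals.
# It outputs grass, chaparral and desert grass, gravel, mud, sand, asphalt, scree and snow layers, plus dynamic snow, desert-grass noise and terrain masks.
# Shared biome and mobile variation masks drive the layer functions, mostly at full weight.
# Gravel feeds mud, desert macro detail feeds sand, and grass accepts a custom slope mask.
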MaterialFunction /Hera/Environment/Terrain/Landscape/Materials/Layers/MF_Hera_StaticMeshLandscapeLayerMatch
graph {
 properties {description=Normals are in worldspace.}
 nodes [
  n1 = MaterialFunctionCall!(
   Material
   MobileMaterial
   IsForStaticMesh
   UseTangentNormals=StaticBool()
   LayerWeight=Constant() {r=1}
   MudWeight
   TerrainMasks_QuadA=NamedRerouteUsage()
   TerrainMasks_QuadB=NamedRerouteUsage()
   TerrainMasks_QuadC=NamedRerouteUsage()
   TerrainMasks_QuadD=NamedRerouteUsage()
   VariationMasks Mobile=NamedRerouteUsage()
  ) {materialFunction=&/MF_Hera_Gravel}
  FunctionOutput(None=MaterialFunctionCall!(
    Material
    MobileMaterial
    MobileMudMaterial=@n1.MobileMaterial
    ForMudMaterial=@n1.ForMudMaterial
    LayerWeight=Constant() {r=1}
    UseTangentNormals=StaticBool()
    TerrainMasks_QuadA=NamedRerouteUsage()
    TerrainMasks_QuadB=NamedRerouteUsage()
    TerrainMasks_QuadC=NamedRerouteUsage()
    TerrainMasks_QuadD=NamedRerouteUsage()
    VariationMasks Mobile=NamedRerouteUsage()
   ) {materialFunction=&/MF_Hera_Mud}.StaticMeshMaterial) {outputName=MudLayer, sortPriority=30}
  n2 = StaticBool()
  n3 = FunctionInput(Preview=StaticBool()) {
   bUsePreviewValueAsDefault=true
   desc=Setting this to true will ignore Grass_EdgeOnly input
   description=Setting this to true will ignore Grass_EdgeOnly input
   inputName=Grass_UseCustomSlopeMask
   inputType=FunctionInput_StaticBool
   sortPriority=13
  }
  n4 = FunctionInput(Preview=Constant()) {bUsePreviewValueAsDefault=true, inputName=Grass_CustomSlopeMask, inputType=FunctionInput_Scalar, sortPriority=14}
  n5 = Constant() {r=1}
  n6 = NamedRerouteUsage()
  n7 = NamedRerouteUsage()
  n8 = NamedRerouteUsage()
  n9 = NamedRerouteUsage()
  n10 = NamedRerouteUsage()
  n11 = NamedRerouteUsage()
  FunctionOutput(None=NamedRerouteDeclaration(None=MaterialFunctionCall!(
     Material
     MobileMaterial
     IsForStaticMesh
     UseTangentNormals=@n2
     UseCustomSlopeMask=@n3
     CustomSlopeMask=@n4
     Grass_EdgeOnly
     LayerWeight=@n5
     TerrainMasks_QuadA=@n6
     TerrainMasks_QuadB=@n7
     TerrainMasks_QuadC=@n8
     TerrainMasks_QuadD=@n9
     VariationMasks Mobile=@n10
     DesertGrass_SimpleNoise=@n11
    ) {materialFunction=&/MF_Hera_Grass}.StaticMeshMaterial) {name=GrassLayer, nodeColor=<rgba 0.5 0 1 1>}) {outputName=GrassLayer}
  n12 = NamedRerouteUsage()
  FunctionOutput(None=MaterialFunctionCall!(
    Material=@n12
    MobileMaterial=@n12
    GrassWeight=Constant() {r=1}
    LayerWeight=NamedRerouteUsage()
    UseTangentNormals=StaticBool()
    TerrainMasks_QuadA=NamedRerouteUsage()
    TerrainMasks_QuadB=NamedRerouteUsage()
    TerrainMasks_QuadC=NamedRerouteUsage()
    TerrainMasks_QuadD=NamedRerouteUsage()
    VariationMasks Mobile=NamedRerouteUsage()
   ) {materialFunction=&/MF_Hera_Grass_Chaparral}.StaticMeshMaterial) {outputName=ChaparralGrassLayer, sortPriority=1}
  n13 = NamedRerouteUsage()
  n14 = MaterialFunctionCall!(
   Material=@n13
   MobileMaterial=@n13
   ChapCoastMobile
   GrassWeight=Constant() {r=1}
   LayerWeight=NamedRerouteUsage()
   UseTangentNormals=StaticBool()
   TerrainMasks_QuadA=NamedRerouteUsage()
   TerrainMasks_QuadB=NamedRerouteUsage()
   TerrainMasks_QuadC=NamedRerouteUsage()
   TerrainMasks_QuadD=NamedRerouteUsage()
   VariationMasks Mobile=NamedRerouteUsage()
  ) {materialFunction=&/MF_Hera_Grass_Desert}
  n15 = NamedRerouteDeclaration(None=@n14.DesertGrass_SimpleNoise) {name=DesertGrass_SimpleNoise, desc=This used to use DesertGrass_SimpleNoise, nodeColor=<rgba 0.5 0 1 1>}
  FunctionOutput(None=@n15) {outputName=DesertGrassSimpleNoise, sortPriority=3}
  FunctionOutput(None=@n1.StaticMeshMaterial) {outputName=GravelLayer, sortPriority=20}
  n16 = MaterialFunctionCall!(BiomeMask=NamedRerouteUsage(), IsForStaticMesh) {materialFunction=/Hera/Environment/Terrain/Landscape/Materials/Layers/Misc/MF_Desert_MacroDetail}
  FunctionOutput(None=MaterialFunctionCall!(
    Material
    MobileMaterial
    LayerWeight=Constant() {r=1}
    UseTangentNormals=StaticBool()
    TerrainMasks_QuadA=NamedRerouteUsage()
    TerrainMasks_QuadB=NamedRerouteUsage()
    TerrainMasks_QuadC=NamedRerouteUsage()
    TerrainMasks_QuadD=NamedRerouteUsage()
    VariationMasks Mobile=NamedRerouteUsage()
    OD_MD_Normals=@n16.Normals
    OD_MD_SedimentLayers=@n16.Normals
    InputMudLayerMask
    InputSeaweedLayerMask
   ) {materialFunction=&/MF_Hera_Sand}.StaticMeshMaterial) {outputName=SandLayer, sortPriority=40}
  FunctionOutput(None=MaterialFunctionCall!(
    Material
    MobileMaterial
    UseTangentNormals=StaticBool()
    LayerWeight=Constant() {r=1}
    TerrainMasks_QuadA=NamedRerouteUsage()
    TerrainMasks_QuadB=NamedRerouteUsage()
    TerrainMasks_QuadC=NamedRerouteUsage()
    TerrainMasks_QuadD=NamedRerouteUsage()
    VariationMasks Mobile=NamedRerouteUsage()
   ) {materialFunction=&/MF_Hera_Asphalt}.StaticMeshMaterial) {outputName=AsphaltLayer, sortPriority=50}
  n17 = MaterialFunctionCall!(
   Material
   MobileMaterial
   IsForStaticMesh
   UseTangentNormals=StaticBool()
   LayerWeight=Constant() {r=1}
   TerrainMasks_QuadA=NamedRerouteUsage()
   TerrainMasks_QuadB=NamedRerouteUsage()
   TerrainMasks_QuadC=NamedRerouteUsage()
   TerrainMasks_QuadD=NamedRerouteUsage()
   CliffMaterialOnly
   CliffLayerMask
   VariationMasks Mobile=NamedRerouteUsage()
  ) {materialFunction=&/MF_Hera_Snow}
  FunctionOutput(None=@n17.StaticMeshMaterial) {outputName=SnowLayer, sortPriority=300}
  FunctionOutput(None=@n17.PostRVT) {outputName=SnowDynamic, sortPriority=310}
  FunctionOutput(None=MaterialFunctionCall!(
    Material
    MobileMaterial
    UseTangentNormals=StaticBool()
    LayerWeight=Constant() {r=1}
    TerrainMasks_QuadA=NamedRerouteUsage()
    TerrainMasks_QuadB=NamedRerouteUsage()
    TerrainMasks_QuadC=NamedRerouteUsage()
    TerrainMasks_QuadD=NamedRerouteUsage()
    VariationMasks Mobile=NamedRerouteUsage()
   ) {materialFunction=&/MF_Hera_Scree}.StaticMeshMaterial) {outputName=ScreeLayer, sortPriority=60}
  FunctionOutput(None=@n14.StaticMeshMaterial) {outputName=DesertGrassLayer, sortPriority=2}
  n18 = MaterialFunctionCall!(UseLandscapeCoords, "UseProjection(translucent)") {materialFunction=/BRCoreMaterials/Hera/Terrain/BiomeMask/MF_Hera_BiomeMasks}
  n19 = MaterialFunctionCall!(
   TerrainMasks_QuadA=FunctionOutput(None=NamedRerouteDeclaration(None=@n18.QuadA_RGBA) {name=TerrainMasks_QuadA, nodeColor=<rgba 0.5 0 1 1>}) {outputName=TerrainMasks_QuadA, sortPriority=-500}
   TerrainMasks_QuadB=FunctionOutput(None=NamedRerouteDeclaration(None=@n18.QuadB_RGBA) {name=TerrainMasks_QuadB, nodeColor=<rgba 0.5 0 1 1>}) {outputName=TerrainMasks_QuadB, sortPriority=-499}
   TerrainMasks_QuadC=FunctionOutput(None=NamedRerouteDeclaration(None=@n18.QuadC_RGBA) {name=TerrainMasks_QuadC, nodeColor=<rgba 0.5 0 1 1>}) {outputName=TerrainMasks_QuadC, sortPriority=-498}
   TerrainMasks_QuadD=FunctionOutput(None=NamedRerouteDeclaration(None=@n18.QuadD_RGBA) {name=TerrainMasks_QuadD, nodeColor=<rgba 0.5 0 1 1>}) {outputName=TerrainMasks_QuadD, sortPriority=-497}
  ) {materialFunction=/BRCoreMaterials/Hera/Terrain/BiomeMask/MF_HeraBiomeMasks_Isolated}
  NamedRerouteDeclaration(None=@n19.03_Desert) {name=BM_Desert, nodeColor=<rgba 0.5 0 1 1>}
  NamedRerouteDeclaration(None=@n19.04_GoldenFall) {name=BM_Alien, nodeColor=<rgba 0.5 0 1 1>}
  NamedRerouteDeclaration(None=MaterialFunctionCall() {materialFunction=/BRCoreMaterials/Hermes/Terrain/Landscape/Materials/MaterialFunctions/MF_Hermes_VariationMasks_Mobile}.VariationMasks Mobile) {name=VariationMasks Mobile, nodeColor=<rgba 1 0.9141 0 1>}
  NamedRerouteDeclaration(None=@n19.02_Archipelago) {name=BM_Boreal, nodeColor=<rgba 0.5 0 1 1>}
  NamedRerouteDeclaration(None=@n19.00_Chaparral) {name=BM_Chaparral, nodeColor=<rgba 0.5 0 1 1>}
  NamedRerouteDeclaration(None=@n19.01_RedWoods) {name=BM_RedWoods, nodeColor=<rgba 0.5 0 1 1>}
  NamedRerouteDeclaration(None=@n19.05_GoldenFall2) {name=BM_Coastal, nodeColor=<rgba 0.5 0 1 1>}
  NamedRerouteDeclaration(None=@n19.06_NarrowFlea) {name=BM_GolfCourse, nodeColor=<rgba 0.5 0 1 1>}
 ]
}
